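# M_UI_CrewLogo_Master is a translucent UI-domain master material for displaying a crew logo texture, defaulting to T_Crew_Workmark_L.
# Width and Height parameters are appended into a texture size for MF_UI_UVScaling, whose "Scaled to Fit" output supplies the logo's UVs.
# The logo is a clamped, alpha-type texture parameter, alongside a constant white color node, with no links to material outputs shown.
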Material /Game/Athena/UI/Logos/Textures/M_UI_CrewLogo_Master
graph {
 properties {
  bAutomaticallySetUsageInEditor=false
  bUsedWithStaticMesh=false
  blendMode=BLEND_Translucent
  materialDomain=MD_UI
  usedShadingModels=Default Lit
 }
 nodes [
  Constant3Vector() {constant=<rgba 1 1 1 0>}
  n1 = ScalarParameter() {defaultValue=540, group=Texture, parameterName="Width (X)", sortPriority=0}
  n2 = ScalarParameter() {defaultValue=160, group=Texture, parameterName="Height (Y)", sortPriority=0}
  TextureSampleParameter2D(UVs=MaterialFunctionCall(TextureSize=AppendVector(A=@n1, B=@n2)) {materialFunction=&MF_UI_UVScaling}.Scaled to Fit) {
   group=Texture
   parameterName=Logo_Texture
   samplerSource=SSM_Clamp_WorldGroupSettings
   samplerType=SAMPLERTYPE_Alpha
   texture=&/T_Crew_Workmark_L
  }
 ]
}
